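# Physics asset for the Zest backpack cosmetic (M_MED_Zest_Pack), giving its sheath and hanging tassels collision and ragdoll-style motion.
# Capsule bodies cover the lower sheath and four dynamic tassel bones, with multiple capsules per bone.
# Two two-link tassel chains, upper and lower, hang from the sheath through constraints with tight swing and twist limits.
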PhysicsAsset /BRCosmetics/Accessories/FORT_Backpacks/Backpack_M_MED_Zest/Meshes/M_MED_Zest_Pack_Physics
bodies {
 dyn_lower_sheath [
  {
   shapeType=Capsule
   length=31.5787
   radius=4.3398
   shapeName=dyn_lower_sheath_capsule
   center=<0.2511 0.0001 -1.6897>
   rotation=<rot -87.2731 0 0>
  }
  {shapeType=Capsule, length=14.9196, radius=5.1042, shapeName=dyn_lower_sheath_capsule, center=<0.2809 0 -1.1251>, rotation=<rot -88.1986 0 0>}
  {shapeType=Capsule, length=15, radius=15, shapeName=dyn_lower_sheath_capsule, center=<1.519 20.0069 0.8772>, rotation=<rot 29.9998 0 0>}
 ]
 dyn_lower_tassel_01 [
  {
   shapeType=Capsule
   length=0.9746
   radius=0.6709
   shapeName=dyn_lower_tassel_01_capsule
   center=<2.6822 0.0331 0>
   rotation=<rot -90 -165.9637 -109.0351>
  }
  {
   shapeType=Capsule
   length=1.2045
   radius=0.5919
   shapeName=dyn_lower_tassel_01_capsule
   center=<1.5585 0.0652 0>
   rotation=<rot 90 0 -10.0092>
  }
 ]
 dyn_lower_tassel_02 [
  {
   shapeType=Capsule
   length=1.7021
   radius=1.1846
   shapeName=dyn_lower_tassel_02_capsule
   center=<3.2583 0 0.3855>
   rotation=<rot -0.0006 -10.0005 89.9967>
  }
  {
   shapeType=Capsule
   length=1.6865
   radius=1.0225
   shapeName=dyn_lower_tassel_02_capsule
   center=<1.1491 0.0902 0>
   rotation=<rot -0.0001 -10.0001 89.9992>
  }
 ]
 dyn_upper_tassel_01 [
  {
   shapeType=Capsule
   length=1.4723
   radius=0.5591
   shapeName=dyn_upper_tassel_01_capsule
   center=<5.5802 -0.0001 0>
   rotation=<rot 90 -90 -179.9986>
  }
  {
   shapeType=Capsule
   length=3.7193
   radius=0.8436
   shapeName=dyn_upper_tassel_01_capsule
   center=<3.171 -0.0001 0.0001>
   rotation=<rot 74.9991 -90.0082 -90.0073>
  }
 ]
 dyn_upper_tassel_02 [
  {shapeType=Capsule, length=2.027, radius=1.121, shapeName=dyn_upper_tassel_02_capsule, center=<5.9296 -0.0001 0>, rotation=<rot 0 0 89.9989>}
  {shapeType=Capsule, length=5.2363, radius=1.2563, shapeName=dyn_upper_tassel_02_capsule, center=<3.2789 -0.0001 0>, rotation=<rot 90 0 0>}
 ]
}
constraints [
 {bone1Name=dyn_upper_tassel_01, bone2Name=dyn_lower_sheath, swing1LimitDegrees=20, swing2LimitDegrees=10, twistLimitDegrees=5}
 {bone1Name=dyn_upper_tassel_02, bone2Name=dyn_upper_tassel_01, swing1LimitDegrees=20, swing2LimitDegrees=10, twistLimitDegrees=5}
 {bone1Name=dyn_lower_tassel_02, bone2Name=dyn_lower_tassel_01, swing1LimitDegrees=20, swing2LimitDegrees=10, twistLimitDegrees=5}
 {bone1Name=dyn_lower_tassel_01, bone2Name=dyn_lower_sheath, swing1LimitDegrees=20, swing2LimitDegrees=10, twistLimitDegrees=10}
]
